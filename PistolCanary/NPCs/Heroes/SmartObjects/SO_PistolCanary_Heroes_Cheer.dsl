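; Smart object Blueprint for a Heroes NPC cheer spot in PistolCanary, kept inactive until explicitly enabled.
; BeginPlay calls the parent, then disables the SmartObject component under the Gameplay reason.
; The EnableSO custom event re-enables it and writes a shipping log message on failure, which names the asset SO_DentVibe_Heroes_Cheer rather than PistolCanary.
; The cheer behavior itself is not defined in this graph.
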(event EventBeginPlay
  (|CallParentFunction)
  (if (SmartObject|SetSmartObjectEnabled(defaultreason:Gameplay) (Variables|Default|GetSmartObject) false :!class /Script/SmartObjectsModule.SmartObjectComponent)
    _))

(event Custom|EnableSO
  (if (not (SmartObject|SetSmartObjectEnabled(defaultreason:Gameplay) (Variables|Default|GetSmartObject) true :!class /Script/SmartObjectsModule.SmartObjectComponent))
    (Fortnite|FortShippingLog self "Smart Object SO_DentVibe_Heroes_Cheer failed to Activate" true)))

(fn ConstructionScript
  (|CallParentFunction))
